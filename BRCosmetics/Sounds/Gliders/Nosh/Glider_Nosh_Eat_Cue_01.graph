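# Sound cue for the eating sound of the Nosh glider cosmetic.
# A Random node picks one of three eat wave variations with equal weight.
# It plays at reduced volume on the glider landing sound class, spatialized through the player glider thrust attenuation.
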SoundCue /BRCosmetics/Sounds/Gliders/Nosh/Glider_Nosh_Eat_Cue_01
root=Random_1
nodes [
 {id=Random_1, children [WavePlayer_2,WavePlayer_5,WavePlayer_6], properties {weights [1,1,1]}}
 {id=WavePlayer_2, properties {soundWaveAssetPtr=&/Glider_Nosh_Eat_03}}
 {id=WavePlayer_5, properties {soundWaveAssetPtr=&/Glider_Nosh_Eat_01}}
 {id=WavePlayer_6, properties {soundWaveAssetPtr=&/Glider_Nosh_Eat_02}}
]
properties {
 attenuationSettings=&Fort_Player_Glider_Thrust_Athena_Att
 maxDistance=12500
 soundClassObject=&Fort_Gliders_Land
 volumeMultiplier=0.35
}
