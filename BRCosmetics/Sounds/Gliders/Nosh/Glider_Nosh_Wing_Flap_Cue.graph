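# Sound cue for the wing flap of the Nosh glider cosmetic.
# A single Random node picks one of four wing recordings with equal weight on each play.
# The cue plays at reduced volume through a glider sound class and uses the shared player glider thrust attenuation for distance falloff.
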SoundCue /BRCosmetics/Sounds/Gliders/Nosh/Glider_Nosh_Wing_Flap_Cue
root=Random_0
nodes [
 {id=Random_0, children [WavePlayer_0,WavePlayer_1,WavePlayer_3,WavePlayer_4], properties {weights [1,1,1,1]}}
 {id=WavePlayer_0, properties {soundWaveAssetPtr=&/Glider_Nosh_Wings_01}}
 {id=WavePlayer_1, properties {soundWaveAssetPtr=&/Glider_Nosh_Wings_02}}
 {id=WavePlayer_3, properties {soundWaveAssetPtr=&/Glider_Nosh_Wings_03}}
 {id=WavePlayer_4, properties {soundWaveAssetPtr=&/Glider_Nosh_Wings_04}}
]
properties {
 attenuationSettings=&Fort_Player_Glider_Thrust_Athena_Att
 maxDistance=12500
 soundClassObject=&Fort_Gliders_Land
 volumeMultiplier=0.25
}
